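# Animation Blueprint for the RoosterRoast female head skeleton, adjusting the head part's pose on a Fortnite player pawn.
# A per-bone layered blend copies clavicle and upper-arm deform bones from Body_Mesh, cached when character customization loads, onto the input pose.
# Fixed Modify Bone nodes then nudge the left upper eyelid and scale both neck bones.
# Each update sets wearingHat by wildcard-matching the hat mesh name, but the AnimGraph never reads it.
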AnimBlueprint /BRCosmetics/Characters/Player/Female/Medium/Heads/F_MED_RoosterRoast/Meshes/CustomSkeleton/Fortnite_Base_Head_RoosterRoast_Skeleton_AnimB
target_skeleton=&/Fortnite_Base_Head_RoosterRoast_Skeleton
graphs [
 AnimGraph {nodes [
   n1 = Layered blend per bone(BasePose=Input Pose(), BlendPoses_0=Copy Pose From Mesh(SourceMeshComponent=Get Body_Mesh())) {layerSetup [{branchFilters [
       {boneName=dfrm_clavicle_l}
       {boneName=dfrm_clavicle_r}
       {boneName=dfrm_upperarm_l}
       {boneName=dfrm_upperarm_r}
       {boneName=dfrm_upperarm_twist_01_l}
       {boneName=dfrm_upperarm_twist_01_r}
      ]}]}
   n2 = "Transform (Modify) Bone - Bone: L_eye_lid_upper_mid"!(
    ComponentPose=Local To Component(LocalPose=@n1)
    bAlphaBoolEnabled:bool=True
    Translation="0,-0.1,0"
    Rotation="-1.451,-8.648,-10.114"
    Scale="1,1,1"
   ) -> (Pose) {
    alphaInputType=Bool
    rotationMode=BMM_Additive
    rotationSpace=BCS_BoneSpace
    translationMode=BMM_Additive
    translationSpace=BCS_BoneSpace
    boneToModify {boneName=L_eye_lid_upper_mid}
    rotation=<rot -1.451 -8.648 -10.114>
    translation {y=-0.1}
   }
   n3 = "Transform (Modify) Bone - Bone: neck_01"!(ComponentPose=@n2, bAlphaBoolEnabled:bool=True, Translation="0,0,0", Rotation="0,0,0", Scale="0.5,0.5,0.5") -> (Pose) {
    alphaInputType=Bool
    rotationMode=BMM_Additive
    rotationSpace=BCS_BoneSpace
    scaleMode=BMM_Additive
    scaleSpace=BCS_BoneSpace
    translationMode=BMM_Additive
    translationSpace=BCS_BoneSpace
    boneToModify {boneName=neck_01}
    scale=<0.5 0.5 0.5>
   }
   n4 = "Transform (Modify) Bone - Bone: neck_02"!(ComponentPose=@n3, bAlphaBoolEnabled:bool=True, Translation="0,0,0", Rotation="0,0,0", Scale="0.5,0.5,0.5") -> (Pose) {
    alphaInputType=Bool
    rotationMode=BMM_Additive
    rotationSpace=BCS_BoneSpace
    scaleMode=BMM_Additive
    scaleSpace=BCS_BoneSpace
    translationMode=BMM_Additive
    translationSpace=BCS_BoneSpace
    boneToModify {boneName=neck_02}
    scale=<0.5 0.5 0.5>
   }
   Output Pose(Result=Component To Local(ComponentPose=@n4))
  ]}
 EventGraph {nodes [
   Event OnCharacterCustomizationLoaded()
   > n1 = Cast To PlayerPawn_Athena(Object=TryGetPawnOwner())
   > Set Body_Mesh(Body_Mesh=GetSkeletalMeshForPartType(self=@n1))
   Event BlueprintUpdateAnimation()
   > n2 = Cast To FortPlayerPawn(Object=TryGetPawnOwner())
   n3 = GetSkeletalMeshForPartType(self=@n2, PartType=Hat)
   > Set wearingHat(wearingHat:bool=MatchesWildcard(SourceString=GetDisplayName(Object=@n3), Wildcard=*HelmetTemplate_Head_Trap_Gadget*, SearchCase=CaseSensitive)) -> (Output_Get:bool)
  ]}
]
variables [
 Body_Mesh {type=SkeletalMeshComponent}
 IsWearingMask {type=boolean}
 wearingHat {type=boolean}
]
